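; Season 29 Battle Pass quest ability that reports repeated dashes, judging by its name, by broadcasting a custom verb message for the avatar.
; On activation it listens on the avatar for the SunRose SilkyWolf Tutorial.End gameplay event.
; CheckEventTime keeps two match timestamps, ignores events under 0.1 seconds apart, and succeeds when a third lands within QuestTimeLength of the first.
; A late event slides the window forward, and ending the ability clears both timestamps.
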(event Ability|EventActivateAbility
  (|CallParentFunction)
  (Ability|Async|WaitGameplayEventToActor (Ability|GetAvatarActorfromActorInfo) "(TagName=\"Gameplay.Status.SunRose.SilkyWolf.Tutorial.End\")"
    (:EventReceived
      (bind _eventsucceeded (CallFunction|CheckEventTime))
      (if _eventsucceeded
        (CustomVerb|BroadcastCustomVerbMessage (CustomVerb|MakeCustomVerbMessage (Ability|GetAvatarActorfromActorInfo) (Variables|Default|GetVerb%20Name)))))))

(event Ability|EventOnEndAbility
  (params (bWasCancelled :type Boolean))
  (Variables|Default|SetEventTime_2 0.0)
  (Variables|Default|SetEventTime_1 0.0))

(fn CheckEventTime
  (bind _returnvalue (Fortnite|GetMatchTime))
  (bind _returnvalue_2 (Fortnite|GetMatchTime))
  (bind _eventtime_2 (Variables|Default|GetEventTime_2))
  (bind _returnvalue_3 (- _returnvalue_2 _eventtime_2))
  (bind _returnvalue_4 (>= _returnvalue_3 0.1))
  (if (== (Variables|Default|GetEventTime_1) 0)
    (Variables|Default|SetEventTime_1 _returnvalue)
    (elif (== (Variables|Default|GetEventTime_2) 0)
      (if (>= (- _returnvalue (Variables|Default|GetEventTime_1)) 0.1)
        (Variables|Default|SetEventTime_2 _returnvalue))
      (elif (and (<= (- (Fortnite|GetMatchTime) (Variables|Default|GetEventTime_1)) (Variables|Default|GetQuestTimeLength)) _returnvalue_4)
        (return true)
        (elif _returnvalue_4
          (Variables|Default|SetEventTime_1 (Variables|Default|GetEventTime_2))
          (Variables|Default|SetEventTime_2 (Fortnite|GetMatchTime))
          (return false)
          (else
            (return false))))))
  (return false))
